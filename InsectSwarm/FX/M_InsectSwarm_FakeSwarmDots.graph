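# Masked, two-sided Niagara mesh-particle material that fakes an insect swarm as animated noise dots on a mesh.
# Panning noise layers build the dark-brown to amber color, a normal-map detail and a dot mask shaped by UV height, view angle and particle alpha.
# Differenced soft noise displaces vertices along normals, plus a signed noise-vector jitter.
# Dynamic parameters drive it: Hostile speeds animation, ByShoot amplifies displacement and overrides particle alpha, Erode offsets the mask.
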Material /InsectSwarm/FX/M_InsectSwarm_FakeSwarmDots
graph {
 properties {bUsedWithNiagaraMeshParticles=true, blendMode=BLEND_Masked, pixelDepthOffsetMode=PDOM_Legacy, twoSided=true, usedShadingModels=Default Lit}
 nodes [
  n1 = TextureCoordinate()
  n2 = Multiply(A=@n1, B=ScalarParameter() {defaultValue=4, parameterName=Tile})
  n3 = Divide(A=@n2)
  n4 = DynamicParameter() {paramNames [Hostile,ByShoot,Erode,Param4]}
  n5 = Add(A=Multiply(A=Time(), B=LinearInterpolate(Alpha=@n4.Hostile) {constA=1, constB=2}), B=ParticleRandom())
  n6 = Divide(A=@n2) {constB=3.5}
  n7 = Subtract(
   A=TextureSample(UVs=Panner(Coordinate=@n3, Time=@n5) {speedX=0.3, speedY=0.3}) {texture=/Game/Effects/Fort_Effects/Emotes/UltraLight/Texture/T_SoftNoise_BC7}.RGB
   B=TextureSample(UVs=Panner(Coordinate=@n6, Time=@n5) {speedX=0.3, speedY=0.3}) {texture=&T_SoftNoise_BC7}.RGB
  )
  n8 = ScalarParameter() {defaultValue=60, parameterName=WPO_Intensity}
  n9 = LinearInterpolate(A=@n8, B=Multiply(A=@n8) {constB=10}, Alpha=DynamicParameter() {paramNames [Hostile,ByShoot,Erode,Param4]}.ByShoot)
  n10 = TextureSample(UVs=Panner(Coordinate=Divide(A=@n6) {constB=4}, Time=@n5) {speedX=0.1, speedY=-0.1}) {texture=&T_SoftNoise_BC7}
  MaterialExpressionAppend3Vector_0 = MaterialXAppend3Vector!(A=LinearInterpolate(Alpha=@n10.R) {constA=-1}, B=LinearInterpolate(Alpha=@n10.G) {constA=-1}, C=LinearInterpolate(Alpha=@n10.B) {constA=-1})
  Add(A=Multiply(A=VertexNormalWS(), B=Multiply(A=@n7, B=@n9)), B=Multiply(A=Divide(A=@n9) {constB=3}, B=@MaterialExpressionAppend3Vector_0))
  Constant()
  n11 = VectorParameter() {parameterName=Param_1, defaultValue=<rgba 0.1615 0.0442 0.0002 1>}
  n12 = VectorParameter() {defaultValue=<rgba 0.0104 0.0034 0.001 1>}
  n13 = Multiply(A=@n5, B=ScalarParameter() {defaultValue=5, parameterName=PanSpeed})
  n14 = TextureSample(UVs=Panner(Coordinate=@n2, Time=@n13) {speedX=-0.2, speedY=-0.2}) {texture=/Game/Global/Textures/Default/Noise/T_TileNoise_02}
  n15 = Panner(Coordinate=@n2, Time=@n13) {speedX=0.2, speedY=0.2}
  n16 = TextureSample(UVs=@n15) {texture=/Game/Effects/Fort_Effects/Textures/Tile/T_TileNoise_01}
  n17 = Multiply(A=@n14.R, B=@n16.R)
  Multiply(A=Multiply(A=LinearInterpolate(A=@n11.RGB, B=@n12.RGB, Alpha=@n17)) {constB=8}, B=@n16.R)
  ScalarParameter() {defaultValue=1, parameterName=Param_3}
  ScalarParameter() {parameterName=Param_4}
  ScalarParameter() {defaultValue=0.8, parameterName=Param_5}
  n18 = Multiply(A=MaterialFunctionCall(Float2=@n1) {materialFunction=&BreakOutFloat2Components}.G) {constB=0.5}
  n19 = Saturate(None=Multiply(A=Sine(None=@n18)) {constB=3})
  n20 = Multiply(
   A=Add(A=@n16.R, B=@n14.R)
   B=MaterialFunctionCall("FadeSensitivity (optional)"=Constant() {r=1}) {materialFunction=&FadeBasedOnViewAngle}.Result
  )
  n21 = Multiply(B=ParticleColor().A) {constA=2.5}
  n22 = DynamicParameter() {paramNames [Hostile,ByShoot,Erode,Param4]}
  Subtract(A=Multiply(A=Multiply(A=@n19, B=@n20), B=LinearInterpolate(A=@n21, Alpha=@n22.ByShoot) {constA=2.5}), B=OneMinus(None=@n22.Erode))
  TextureSample(UVs=Multiply(A=@n15) {constB=3}) {samplerType=SAMPLERTYPE_Normal, texture=/Game/Characters/Player/Female/Medium/Bodies/F_MED_Bistro_Astronaut/FX/T_GaussianNoise_Normal}
 ]
}
