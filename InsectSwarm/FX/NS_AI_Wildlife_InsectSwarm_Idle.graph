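# Niagara idle effect for a wildlife insect swarm that shifts between flower and hostile looks.
# GPU and CPU insect emitters combine curl noise with pulsing point attraction, alongside lights, ground dust and distance-culled mesh fake-swarm trails.
# User Flower and Hostile values blend motion and color, with time-of-day dimming.
# A system timeline weakens attraction and drives a force at ShootPosition, while AttractToHand pulls insects into the collecting hand.
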NiagaraSystem /InsectSwarm/FX/NS_AI_Wildlife_InsectSwarm_Idle
dependencies {
 usedDataInterfaces [&NiagaraDataInterfaceCurve, &NiagaraDataInterfaceArrayFloat]
 usedDynamicInputs [&Abs_Float, &OneMinusFloat, &LerpLinearColors]
 usedModules [
  &self:SystemSpawnScript.SystemScriptSource.SystemScriptGraph.NiagaraNodeAssignment_1.SetVariables_#1
  &self:SystemSpawnScript.SystemScriptSource.SystemScriptGraph.NiagaraNodeAssignment_3.SetVariables_#2
  &SystemState
  &self:SystemSpawnScript.SystemScriptSource.SystemScriptGraph.NiagaraNodeAssignment_2.SetVariables_#3
  &self:SystemSpawnScript.SystemScriptSource.SystemScriptGraph.NiagaraNodeAssignment_4.SetVariables_#4
  &self:SystemSpawnScript.SystemScriptSource.SystemScriptGraph.NiagaraNodeAssignment_0.SetVariables_#5
  &Timeline
 ]
}
emitters {
 FakeSawrm_LOW {
  emitterUpdateScript [
   &EmitterState(
    Enable Distance Culling=true
    Enable Visibility Culling=true
    Inactive Response=::NewEnumerator1
    Life Cycle Mode=::NewEnumerator1
    Loop Behavior=::NewEnumerator1
    Loop Duration Mode=::NewEnumerator1
    Max Distance Response=::SleepAndClearParticles
    MaxDistance=3900
    MinDistance=2000
    Module.bUseMaxDistance=true
    Module.bUseMinDistance=false
    Reset Age On Awaken=true
    Scalability Mode=::NewEnumerator1
    UseLoopDelay=false
    Visibility Cull Response:ENiagaraExecutionStateManagement=::SleepAndClearParticles
   )
   &SpawnBurst_Instantaneous(Use Spawn Probability=false, UseLoopCountLimit=false)
   /SetVariables(Emitter.WPO:NiagaraFloat=&Lerp_Float(A=50, B=25, Alpha=$User.Flower)) {isSetParameters=true}
  ]
  particleSpawnScript [
   &InitializeParticle(Lifetime=1.2, Mesh Scale=<2.5 1.2 1.2>, Mesh Scale Mode=::NewEnumerator4, UsePositionOffset=false, Write Material Random=true)
   &InitialMeshOrientation(Mesh Axis To Orient=<0 1 0>, Mesh Orientation Mode=::NewEnumerator2, Use Rotation Vector=false)
  ]
  particleUpdateScript [
   &ParticleState(Kill Particles When Lifetime Has Elapsed=false, Let Infinitely Lived Particles Die When Emitter Deactivates=true)
   &ScaleMeshSize(Scale Factor=&VectorFromFloat(Value=3))
   &ScaleMeshSize(Scale Factor=&Lerp_Vector(A=<1 1 1>, B=<0.5 1 1>, Alpha=$User.Flower)) {module=ScaleMeshSize001}
   &InterpolateOverTime(Rate Of Change=2, Target Float Value:NiagaraFloat=$System.Timeline.Float)
   &DynamicMaterialParameters(
    Index 0 Param 1=$User.Hostile
    Index 0 Param 2=&Lerp_Float(A=0, B=1, Alpha=$StackContext.InterpolateOverTime.InterpolateOverTime_Float.Moving Average)
    Index 0 Param 3=1
    Module.Param0WriteEnabled=true
    Module.Param1WriteEnabled=true
    Module.Param2WriteEnabled=true
    Module.Param3WriteEnabled=false
   )
   &ScaleColor(
    Scale Alpha=&ClampFloat(
     Float=&FN_LOD_Distance_Remap(Start Distance=200, End Distance=2000, Minimum Output=0, Maximum Output=1.8)
     Max=1.8
    )
    ScaleA=true
    ScaleRGB=true
   )
  ]
  rendererClasses [&NiagaraMeshRendererProperties]
  renderers [{rendererClass=&NiagaraMeshRendererProperties, rendererIndex=0}]
 }
 FakeSawrm_Trail_LOW {
  emitterUpdateScript [
   &EmitterState(
    Enable Distance Culling=true
    Enable Visibility Culling=true
    Max Distance Response=::SleepAndClearParticles
    MaxDistance=3900
    Module.bUseMaxDistance=true
    Module.bUseMinDistance=false
    Reset Age On Awaken=true
    Scalability Mode=::NewEnumerator1
   )
   &SpawnPerUnit(Module.Use Max Movement Threshold=true, Module.Use Movement Tolerance=false, Spawn Spacing=250, Use Spawn Probability=false)
  ]
  particleSpawnScript [
   &InitializeParticle(
    Lifetime Min=1.5
    Lifetime Mode=::NewEnumerator1
    Mesh Scale=<1 1.2 1.2>
    Mesh Scale Mode=::NewEnumerator4
    UsePositionOffset=false
    Write Material Random=true
   )
   &InitialMeshOrientation(Use Rotation Vector=true)
  ]
  particleUpdateScript [
   &ParticleState()
   &ScaleMeshSize(Scale Factor=&VectorFromFloat(Value=&FloatFromCurve(FloatCurve=DataInterface() {propertyValues {Curve {keys [
          {tangentMode=RCTM_User, time=0, value=1}
          {arriveTangent=0.183, interpMode=RCIM_Cubic, leaveTangent=0.183, tangentMode=RCTM_User, time=0.3, value=1.025}
          {arriveTangent=0.3351, interpMode=RCIM_Cubic, leaveTangent=0.3351, tangentMode=RCTM_User, time=1, value=0}
         ]}}}))) {module=ScaleMeshSize001}
   &ScaleMeshSize(Scale Factor=&VectorFromFloat(Value=3))
   &InterpolateOverTime(Rate Of Change=2, Target Float Value:NiagaraFloat=$System.Timeline.Float)
   &InheritVelocity(Inherited Velocity Amount Scale=<0.2 0.2 0.2>, Speed Limit=true, UseSpeedThreshold=false)
   &Drag(Drag=4, Use Linear Drag=true, Use Rotational Drag=true)
   &SolveForcesAndVelocity(Clamp Velocity=false, Limit Acceleration=false)
   &DynamicMaterialParameters(
    Index 0 Param 1=$User.Hostile
    Index 0 Param 2=&Lerp_Float(A=0, B=1, Alpha=$StackContext.InterpolateOverTime.InterpolateOverTime_Float.Moving Average)
    Index 0 Param 3=&FloatFromCurve(FloatCurve=DataInterface() {propertyValues {Curve {keys [
         {interpMode=RCIM_Cubic, tangentMode=RCTM_User, time=0, value=0}
         <at 0.1 1>
         {interpMode=RCIM_Cubic, tangentMode=RCTM_User, time=0.7, value=1.05}
         {tangentMode=RCTM_User, time=1, value=0}
        ]}}})
    Module.Param0WriteEnabled=true
    Module.Param1WriteEnabled=true
    Module.Param2WriteEnabled=true
    Module.Param3WriteEnabled=false
   )
   &ScaleColor(
    Scale Alpha=&ClampFloat(
     Float=&FN_LOD_Distance_Remap(Start Distance=200, End Distance=2000, Minimum Output=0, Maximum Output=1.4)
     Max=1.4
    )
    ScaleA=true
    ScaleRGB=true
   )
  ]
  rendererClasses [&NiagaraMeshRendererProperties]
  renderers [{rendererClass=&NiagaraMeshRendererProperties, rendererIndex=0}]
 }
 GroundDust {
  emitterUpdateScript [
   &EmitterState()
   &SpawnRate(
    Spawn Probability=&Lerp_Float(A=&Lerp_Float(A=0.3, B=0, Alpha=$User.Flower), B=1, Alpha=$User.Hostile)
    SpawnRate=40
    Use Spawn Probability=true
   )
   /SetVariables(Emitter.SizeRad=50) {isSetParameters=true}
  ]
  particleSpawnScript [
   /SetVariables(Particles.SizeMax=400, Particles.SizeMin=250) {isSetParameters=true}
   /SetVariables(Particles.SizeCurve:NiagaraFloat=&Lerp_Float(
     A=$Particles.SizeMin
     B=$Particles.SizeMax
     Alpha=&FloatFromCurve(
      FloatCurve=DataInterface() {propertyValues {Curve {keys [
          {interpMode=RCIM_Cubic, tangentMode=RCTM_Break, time=0, value=0}
          {arriveTangentWeight=0.9217, interpMode=RCIM_Cubic, tangentMode=RCTM_Break, time=1, value=1}
         ]}}}
      CurveIndex=&RandomRangeFloat(Minimum=0, Maximum=1)
     )
    )) {isSetParameters=true}
   /SetVariables(Particles.InitSpriteRotRate=&UniformRangedFloat(Minimum=-1, Maximum=1)) {isSetParameters=true}
   &InitializeParticle(
    Color Maximum=<rgba 0.3229 0.3229 0.3229 1>
    Color Minimum=<rgba 0.2396 0.2396 0.2396 1>
    Color Mode=::NewEnumerator2
    Lifetime Max=1.5
    Lifetime Mode=::NewEnumerator1
    Sprite Rotation Mode=::NewEnumerator2
    Sprite Size Mode=::NewEnumerator3
    Uniform Sprite Size=$Particles.SizeCurve
    UsePositionOffset=false
    Write Material Random=true
   )
   &ShapeLocation(
    Handle Radius=30
    MaskSpawnGroup=false
    Offset Mode=::NewEnumerator0
    Override Seed=false
    Rotation Axis=<0 0 1>
    Rotation Mode=::NewEnumerator1
    Shape Origin=$User.DustLocation
    Shape Primitive=::NewEnumerator3
    UseOwnerScale=true
   )
   /SetVariables(Particles.PosSize:NiagaraFloat=&VectorLength(Vector=&ConvertPositionToVector(Input Position=$Particles.Position, Convert To Absolute Large World Vector=false))) {isSetParameters=true}
   &AddVelocity(
    Module.Use Falloff Distance=false
    Override Seed=false
    Velocity Mode=::NewEnumerator1
    Velocity Origin=$User.DustLocation
    Velocity Speed=&Multiply_Float(
     A=&Lerp_Float(
      A=500
      B=1166.6666
      Alpha=&FloatFromCurve(
       FloatCurve=DataInterface() {propertyValues {Curve {keys [
           {interpMode=RCIM_Cubic, tangentMode=RCTM_Break, time=0, value=0}
           {arriveTangent=2.15, interpMode=RCIM_Cubic, tangentMode=RCTM_Break, time=1, value=1}
          ]}}}
       CurveIndex=&RandomRangeFloat(Minimum=0, Maximum=1)
      )
     )
     B=&RandomRangeFloat(Minimum=1, Maximum=1.2)
    )
   )
   &AddVelocity(Velocity=&RandomRangeVector(Minimum=<0 0 150>, Maximum=<0 0 300>), Velocity Speed Scale=0.5) {module=AddVelocity001}
  ]
  particleUpdateScript [
   &ParticleState()
   &Drag(
    Drag=&FloatFromCurve(
     FloatCurve=DataInterface() {propertyValues {Curve {keys [{interpMode=RCIM_Cubic, tangentMode=RCTM_Break, time=0, value=1}, {interpMode=RCIM_Cubic, tangentMode=RCTM_Break, time=0.8, value=0}]}}}
     Scale Curve=4
    )
    Use Linear Drag=true
    Use Rotational Drag=false
   )
   &SolveForcesAndVelocity(Clamp Velocity=false, Limit Acceleration=false)
   &SpriteRotationRate(Rotation Rate=&Multiply_Float(
     A=$Particles.InitSpriteRotRate
     B=&FloatFromCurve(
      FloatCurve=DataInterface() {propertyValues {Curve {keys [
          {arriveTangent=-4.0107, interpMode=RCIM_Cubic, leaveTangent=-4.0107, tangentMode=RCTM_User, time=-0.0046, value=1.0113}
          {arriveTangent=-0.2496, interpMode=RCIM_Cubic, leaveTangent=-0.2496, tangentMode=RCTM_User, time=0.5, value=0.16}
         ]}}}
      Scale Curve=75
     )
    ))
   &ScaleSpriteSize(
    Scale Sprite Size Mode=::NewEnumerator0
    Uniform Scale Factor=&FloatFromCurve(FloatCurve=DataInterface() {propertyValues {Curve {keys [
         {interpMode=RCIM_Cubic, leaveTangent=1.2626, tangentMode=RCTM_Break, time=0, value=0.5}
         {arriveTangent=0.7184, interpMode=RCIM_Cubic, leaveTangent=0.7184, tangentMode=RCTM_User, time=0.3, value=0.7}
         {arriveTangentWeight=0.6265, interpMode=RCIM_Cubic, tangentMode=RCTM_Break, time=1, value=1}
        ]}}})
   )
   &DynamicMaterialParameters(
    Index 0 Param 1=&Add_Float(
     A=&RandomRangeFloat(Minimum=0, Maximum=1)
     B=&FloatFromCurve(
      FloatCurve=DataInterface() {propertyValues {Curve {keys [
          {arriveTangent=2.7533, interpMode=RCIM_Cubic, leaveTangent=1.8958, tangentMode=RCTM_User, time=0, value=1}
          {arriveTangent=0.8292, interpMode=RCIM_Cubic, leaveTangent=0.8292, tangentMode=RCTM_User, time=0.321, value=1.3453}
          {arriveTangent=0.0529, interpMode=RCIM_Cubic, leaveTangent=0.2021, tangentMode=RCTM_User, time=1, value=1.5}
         ]}}}
      CurveIndex=&Add_Float(A=&Multiply_Float(A=$Particles.Initial.MaterialRandom, B=0.5), B=$Particles.NormalizedAge)
      Scale Curve=&Lerp_Float(
       A=0.25
       B=0.7
       Alpha=&NormalizeFloat(Value=$Particles.SizeCurve, Min=$Particles.SizeMin, Max=$Particles.SizeMax)
      )
     )
    )
    Index 0 Param 2=&FloatFromCurve(FloatCurve=DataInterface() {propertyValues {Curve {keys [<at 0 1>,<at 1 0>]}}})
    Index 0 Param 3=&FloatFromCurve(
     FloatCurve=DataInterface() {propertyValues {Curve {keys [
         {interpMode=RCIM_Cubic, leaveTangent=16.5652, leaveTangentWeight=0.4385, tangentMode=RCTM_Break, time=0, value=0}
         {arriveTangentWeight=0.0333, interpMode=RCIM_Cubic, leaveTangentWeight=0.9477, tangentMode=RCTM_Break, time=0.1, value=1}
         {arriveTangent=-1.7159, interpMode=RCIM_Cubic, leaveTangent=-1.7159, tangentMode=RCTM_User, time=0.3, value=0.2}
         {arriveTangent=0.2347, interpMode=RCIM_Cubic, leaveTangent=0.2723, tangentMode=RCTM_Break, time=0.6, value=0}
        ]}}}
     Scale Curve=0
    )
    Index 0 Param 4=&FloatFromCurve(FloatCurve=DataInterface() {propertyValues {Curve {keys [{interpMode=RCIM_Cubic, tangentMode=RCTM_Break, time=0.5, value=1}, {interpMode=RCIM_Cubic, tangentMode=RCTM_Break, time=1, value=0}]}}})
    Index 1 Param 1=&FloatFromCurve(
     FloatCurve=DataInterface() {propertyValues {Curve {keys [
         {interpMode=RCIM_Cubic, leaveTangentWeight=0.0833, tangentMode=RCTM_Break, time=0, value=0}
         {arriveTangentWeight=0.0873, interpMode=RCIM_Cubic, tangentMode=RCTM_Break, time=0.25, value=1}
        ]}}}
     Scale Curve=&Lerp_Float(
      A=0.5
      B=0.75
      Alpha=&NormalizeFloat(Value=$Particles.SizeCurve, Min=$Particles.SizeMin, Max=$Particles.SizeMax)
     )
    )
    Index 1 Param 2=&FloatFromCurve(
     FloatCurve=DataInterface() {propertyValues {Curve {keys [
         {interpMode=RCIM_Cubic, tangentMode=RCTM_Break, time=0, value=0.2}
         {interpMode=RCIM_Cubic, leaveTangent=0.9066, tangentMode=RCTM_Break, time=1, value=2}
        ]}}}
     Scale Curve=&RandomRangeFloat(Minimum=0.1, Maximum=0.2)
    )
    Index 1 Param 3=&FloatFromCurve(FloatCurve=DataInterface() {propertyValues {Curve {keys [
         {interpMode=RCIM_Cubic, leaveTangent=6.3917, tangentMode=RCTM_Break, time=0, value=0}
         {arriveTangent=1.7549, interpMode=RCIM_Cubic, leaveTangent=10.6767, tangentMode=RCTM_User, time=0.25, value=1}
        ]}}})
    Index 1 Param 4=&FloatFromCurve(
     FloatCurve=DataInterface() {propertyValues {Curve {keys [{arriveTangent=-0.136, interpMode=RCIM_Cubic, leaveTangent=-0.136, tangentMode=RCTM_User, time=0, value=1}, <at 1 0.95>]}}}
     Scale Curve=&Lerp_Float(
      A=2
      B=5
      Alpha=&NormalizeFloat(Value=$Particles.SizeCurve, Min=$Particles.SizeMin, Max=$Particles.SizeMax)
     )
    )
    Index 2 Param 1=&FloatFromCurve(
     FloatCurve=DataInterface() {propertyValues {Curve {keys [
         {interpMode=RCIM_Cubic, leaveTangent=4.2485, tangentMode=RCTM_Break, time=0, value=0}
         {arriveTangent=1.0548, interpMode=RCIM_Cubic, leaveTangent=1.0548, tangentMode=RCTM_User, time=0.17, value=0.5715}
         <at 1 1>
        ]}}}
     Scale Curve=3
    )
    Index 2 Param 2=&FloatFromCurve(FloatCurve=DataInterface() {propertyValues {Curve {keys [
         {interpMode=RCIM_Cubic, leaveTangentWeight=0.5024, tangentMode=RCTM_User, time=0, value=1}
         {arriveTangentWeight=0.4857, interpMode=RCIM_Cubic, tangentMode=RCTM_User, time=1, value=1}
        ]}}})
    Index 2 Param 3=&FloatFromCurve(
     FloatCurve=DataInterface() {propertyValues {Curve {keys [
         {interpMode=RCIM_Cubic, leaveTangent=-19.1446, leaveTangentWeight=0.1411, tangentMode=RCTM_Break, time=0, value=1}
         {interpMode=RCIM_Cubic, tangentMode=RCTM_Break, time=0.1, value=0}
         {interpMode=RCIM_Cubic, leaveTangent=11.1233, tangentMode=RCTM_Break, time=1, value=1}
        ]}}}
     Scale Curve=5
    )
    Index 2 Param 4=&RandomRangeFloat(Minimum=1, Maximum=1.75)
    Module.Param0WriteEnabled=true
    Module.Param0WriteEnabled1=true
    Module.Param0WriteEnabled2=true
    Module.Param1WriteEnabled=false
    Module.Param1WriteEnabled1=true
    Module.Param1WriteEnabled2=true
    Module.Param2WriteEnabled=true
    Module.Param2WriteEnabled1=true
    Module.Param2WriteEnabled2=true
    Module.Param3WriteEnabled=true
    Module.Param3WriteEnabled1=true
    Module.Param3WriteEnabled2=true
    Write Parameter Index 1=true
    Write Parameter Index 2=true
   )
   &ScaleColor(
    Scale Alpha=&FloatFromCurve(
     FloatCurve=DataInterface() {propertyValues {Curve {keys [
         {arriveTangent=2.4725, interpMode=RCIM_Cubic, leaveTangent=2.4725, tangentMode=RCTM_User, time=0, value=0}
         {arriveTangent=-2.5249, interpMode=RCIM_Cubic, leaveTangent=-2.5249, tangentMode=RCTM_User, time=0.5, value=0.7}
         {arriveTangent=-0.0405, interpMode=RCIM_Cubic, leaveTangent=-0.0405, tangentMode=RCTM_User, time=1, value=0}
        ]}}}
     Scale Curve=&RandomRangeFloat(Minimum=0.6, Maximum=0.7)
    )
    ScaleA=true
    ScaleRGB=false
   )
   &ScaleColor(Scale Alpha=&Lerp_Float(A=0.5, B=1, Alpha=$User.Hostile), ScaleA=true, ScaleRGB=true) {module=ScaleColor001}
  ]
  rendererClasses [&NiagaraSpriteRendererProperties]
  renderers [
   {rendererClass=&NiagaraSpriteRendererProperties}
   {rendererClass=&NiagaraSpriteRendererProperties, rendererIndex=1}
  ]
 }
 InsectSwarm_Idle {
  simTarget=GPUComputeSim
  emitterUpdateScript [
   &EmitterState(Inactive Response=::NewEnumerator1, Life Cycle Mode=::NewEnumerator1, UseLoopDelay=false)
   &SpawnRate(
    Spawn Probability=&Lerp_Float(A=1, B=1, Alpha=$System.SizeDistanceScalar)
    SpawnRate=60
    Use Spawn Probability=true
   )
   &SpawnRate(SpawnGroup=1, SpawnRate=5, Use Spawn Probability=false) {module=SpawnRate001}
  ]
  particleSpawnScript [
   &InitializeParticle(
    Color Mode=::NewEnumerator0
    Lifetime=&Multiply_Float(
     A=&RandomRangeFloat(Minimum=2, Maximum=3)
     B=&Lerp_Float(A=1, B=0.7, Alpha=$System.Hostile)
    )
    Mesh Scale Mode=::NewEnumerator1
    Mesh Uniform Scale Max=0.4
    Mesh Uniform Scale Min=0.2
    Sprite Size Mode=::NewEnumerator3
    Sprite UV Mode=::NewEnumerator1
    Uniform Sprite Size=45
    UsePositionOffset=false
    Write Material Random=true
   )
   /SetVariables(Particles.RandomInsectTexture:NiagaraFloat=&UniformAOrBFloat(A=0)) {isSetParameters=true, module=SetVariables_#6}
   /SetVariables(Particles.WaspSpriteSize:Vector2f=&Lerp_Vector2(A=<5 5>, B=<8 8>, Alpha=&RandomRangeFloat(Minimum=0.3, Maximum=1))) {isSetParameters=true, module=SetVariables_#7}
   /SetVariables(Particles.MeshVisbilityIndex:NiagaraInt32=&SpawnGroup()) {isSetParameters=true, module=SetVariables_#8}
   &ShapeLocation(
    MaskSpawnGroup=false
    Override Seed=false
    Sphere Radius=&RandomRangeFloat(Minimum=80, Maximum=120)
    UseOwnerScale=true
   )
  ]
  particleUpdateScript [
   &ParticleState()
   &CurlNoiseForce(
    Noise Frequency=12
    Noise Strength=&Lerp_Float(A=&Lerp_Float(A=350, B=100, Alpha=$User.Flower), B=3000, Alpha=$User.Hostile)
    Pan Noise=true
    Pan Noise Field=&Lerp_Vector(A=<1 1 1>, B=<0.2 0.2 0.2>, Alpha=$User.Flower)
   )
   &CurlNoiseForce(
    Noise Frequency=0.5
    Noise Strength=&Lerp_Float(A=500, B=0, Alpha=$User.Flower)
    Pan Noise=true
    Pan Noise Field=<0.3 0.3 0.3>
   ) {module=CurlNoiseForce001}
   &PointAttractionForce(
    Attraction Radius=800
    Attraction Strength=&Multiply_Float(
     A=&Multiply_Float(
      A=&RandomRangeFloat(Minimum=200, Maximum=250)
      B=&Lerp_Float(A=1, B=0.12, Alpha=$User.Flower)
     )
     B=&Multiply_Float(
      A=&Multiply_Float(
       A=&Lerp_Float(
        A=0.2
        B=0.6
        Alpha=&Waveform("Waveform [1]"=::NewEnumerator9, "[1] Frequency"=0.4, "[1] Amplitude Min/Max"=<0 1>)
       )
       B=&Lerp_Float(A=0.7, B=4, Alpha=$User.Hostile)
      )
      B=&Lerp_Float(
       A=1
       B=0
       Alpha=&FloatFromCurve(
        FloatCurve=DataInterface() {propertyValues {Curve {keys [
            {arriveTangent=5.3671, interpMode=RCIM_Cubic, leaveTangent=5.3671, tangentMode=RCTM_User, time=0, value=0}
            {arriveTangent=1.3895, interpMode=RCIM_Cubic, leaveTangent=1.3895, tangentMode=RCTM_User, time=0.1574, value=0.7636}
            {arriveTangent=0.4231, interpMode=RCIM_Cubic, leaveTangent=0.4231, tangentMode=RCTM_User, time=1, value=1}
           ]}}}
        CurveIndex=$System.Timeline.Time
       )
      )
     )
    )
    Falloff Exponent=0.6
    Kill Within Radius=false
    Module.Use Falloff=true
    UseAttractorOffset=true
   )
   &PointForce(
    Force Origin:NiagaraPosition=$User.ShootPosition
    Force Strength=&Lerp_Float(A=0, B=5000, Alpha=$System.Timeline.Float)
    Module.Use Falloff=false
    Module.UseLinearForce=false
   )
   &InterpolateOverTime(Rate Of Change=0.5, Target Float Value:NiagaraFloat=$User.Flower)
   &ScaleColor(
    Scale Alpha=0.4
    Scale RGB=&Lerp_Vector(
     A=&Lerp_Vector(A=<0 0 0>, B=<0 0.2 0.4>, Alpha=$StackContext.InterpolateOverTime.InterpolateOverTime_Float.Moving Average)
     B=<0.5 0 0>
     Alpha=$System.Hostile
    )
    ScaleA=true
    ScaleRGB=true
   )
   &ScaleColor(
    Scale RGB=&VectorFromFloat(Value=&FNTimeOfDayScalar(Night Scale=0.24, Morning Scale=0.4, Day Scale=0.9, Evening Scale=0.6))
    ScaleA=true
    ScaleRGB=true
   ) {module=ScaleColor001}
   &Drag(Use Linear Drag=true, Use Rotational Drag=false)
   &InterpolateOverTime(Rate Of Change=4, Target Float Value:NiagaraFloat=$User.AttractToHand) {module=InterpolateOverTime001}
   &ScaleMeshSize(Scale Factor=&VectorFromFloat(Value=&FloatFromCurve(
      FloatCurve=DataInterface() {propertyValues {Curve {keys [
          {interpMode=RCIM_Cubic, tangentMode=RCTM_User, time=0, value=0}
          <at 0.3 1>
          {interpMode=RCIM_Cubic, tangentMode=RCTM_User, time=0.9, value=1}
          {tangentMode=RCTM_User, time=1, value=0}
         ]}}}
      Scale Curve=&Lerp_Float(A=4, B=0, Alpha=$StackContext.InterpolateOverTime001.InterpolateOverTime_Float.Moving Average)
     )))
   &ScaleSpriteSize(
    Uniform Curve Scale=&Multiply_Float(
     A=&Lerp_Float(A=1, B=5, Alpha=$System.SizeDistanceScalar)
     B=&Lerp_Float(A=1, B=0, Alpha=$StackContext.InterpolateOverTime001.InterpolateOverTime_Float.Moving Average)
    )
    Uniform Curve Sprite Scale=DataInterface() {propertyValues {Curve {keys [
        {interpMode=RCIM_Cubic, tangentMode=RCTM_User, time=0.2, value=0}
        <at 0.3 1>
        {interpMode=RCIM_Cubic, tangentMode=RCTM_User, time=0.9, value=1}
        {tangentMode=RCTM_User, time=1, value=0}
       ]}}}
   )
   /SetVariables(Particles.WaspSpriteSize:Vector2f=&Multiply_Vector2D(
     A=$Particles.Initial.WaspSpriteSize
     B=&Multiply_Vector2D_ByFloat(
      Float=&FloatFromCurve(
       FloatCurve=DataInterface() {propertyValues {Curve {keys [
           <at 0 0 RCIM_Cubic>
           {arriveTangent=0.2423, interpMode=RCIM_Cubic, leaveTangent=0.2423, tangentMode=RCTM_User, time=0.2, value=1}
           {arriveTangent=0.0664, interpMode=RCIM_Cubic, leaveTangent=0.0664, tangentMode=RCTM_User, time=0.8, value=1}
           <at 1 0 RCIM_Cubic>
          ]}}}
       Scale Curve=&Lerp_Float(A=1, B=1.5, Alpha=$System.Hostile)
      )
      Vector2D=&Multiply_Vector2D_ByFloat(
       Float=&Lerp_Float(A=&RandomRangeFloat(Minimum=1.5, Maximum=3), B=4, Alpha=$System.SizeDistanceScalar)
       Vector2D=&Multiply_Vector2D_ByFloat(
        Float=&Lerp_Float(A=1, B=0, Alpha=$StackContext.InterpolateOverTime001.InterpolateOverTime_Float.Moving Average)
        Vector2D=<1 1>
       )
      )
     )
    )) {isSetParameters=true}
   /Niagara/Modules/Update/Utility/LerpParticleAttributes(
    Lerp Alpha:NiagaraFloat=$StackContext.InterpolateOverTime001.InterpolateOverTime_Float.Moving Average
    Lerp Position=true
    Position B:NiagaraPosition=$User.CollectHandPosition
   )
   &SolveForcesAndVelocity(Clamp Velocity=false, Limit Acceleration=false)
   &DynamicMaterialParameters(
    Index 0 Param 1=$User.Flower
    Index 0 Param 2=&Lerp_Float(A=1, B=0.4, Alpha=$User.Flower)
    Index 0 Param 4=$Particles.RandomInsectTexture
    Module.Param0WriteEnabled=true
    Module.Param1WriteEnabled=true
    Module.Param2WriteEnabled=true
    Module.Param3WriteEnabled=true
   ) {module=DynamicMaterialParameters001}
   &KillParticles(Kill Particles=&SetBoolByFloatComparison(A=$Particles.InterpolateOverTime001.InterpolateOverTime_Float.Moving Average, B=0.5))
  ]
  rendererClasses [&NiagaraMeshRendererProperties, &NiagaraSpriteRendererProperties]
  renderers [
   {rendererClass=&NiagaraMeshRendererProperties, rendererIndex=0}
   {rendererClass=&NiagaraSpriteRendererProperties, rendererIndex=1}
   {rendererClass=&NiagaraSpriteRendererProperties, rendererIndex=2}
  ]
 }
 InsectSwarm_Idle001 {
  emitterUpdateScript [
   &EmitterState()
   &SpawnRate(Use Spawn Probability=false) {module=SpawnRate001}
  ]
  particleSpawnScript [
   &InitializeParticle(
    Color Mode=::NewEnumerator0
    Lifetime=2
    Mesh Scale Mode=::NewEnumerator1
    Mesh Uniform Scale Max=0.4
    Mesh Uniform Scale Min=0.1
    Sprite Size Mode=::NewEnumerator3
    Uniform Sprite Size=100
    UsePositionOffset=false
    Write Material Random=true
   )
   &ShapeLocation(
    MaskSpawnGroup=false
    Override Seed=false
    Sphere Radius=&RandomRangeFloat(Minimum=80, Maximum=120)
    UseOwnerScale=true
   )
  ]
  particleUpdateScript [
   &ParticleState()
   &CurlNoiseForce(Noise Frequency=12, Noise Strength=500, Pan Noise=true, Pan Noise Field=<1 1 1>)
   &CurlNoiseForce(Noise Frequency=0.5, Noise Strength=500, Pan Noise=true, Pan Noise Field=<0.3 0.3 0.3>) {module=CurlNoiseForce001}
   &PointAttractionForce(
    Attraction Radius=800
    Attraction Strength=&Multiply_Float(
     A=&RandomRangeFloat(Minimum=200, Maximum=250)
     B=&Lerp_Float(
      A=0.2
      B=0.6
      Alpha=&Waveform("Waveform [1]"=::NewEnumerator9, "[1] Frequency"=0.4, "[1] Amplitude Min/Max"=<0 1>)
     )
    )
    Falloff Exponent=0.6
    Kill Within Radius=false
    Module.Use Falloff=true
    UseAttractorOffset=true
   )
   &ScaleSpriteSize(Uniform Curve Sprite Scale=DataInterface() {propertyValues {Curve {keys [
        {interpMode=RCIM_Cubic, tangentMode=RCTM_User, time=0.2, value=0}
        <at 0.3 1>
        {interpMode=RCIM_Cubic, tangentMode=RCTM_User, time=0.9, value=1}
        {tangentMode=RCTM_User, time=1, value=0}
       ]}}})
   &ScaleMeshSize(Scale Factor=&VectorFromFloat(Value=&FloatFromCurve(
      FloatCurve=DataInterface() {propertyValues {Curve {keys [
          {interpMode=RCIM_Cubic, tangentMode=RCTM_User, time=0, value=0}
          <at 0.3 1>
          {interpMode=RCIM_Cubic, tangentMode=RCTM_User, time=0.9, value=1}
          {tangentMode=RCTM_User, time=1, value=0}
         ]}}}
      Scale Curve=&MaskFloatBySpawnGroup(Masked Float=1, Passthrough Float=&RandomRangeFloat(Minimum=1, Maximum=3))
     )))
   &ScaleColor(
    Scale RGB=&Lerp_Vector(A=&Lerp_Vector(A=<0 0 0>, B=<0 0.1 0.2>, Alpha=$User.Flower), B=<0.5 0 0>, Alpha=$System.Hostile)
    ScaleA=true
    ScaleRGB=true
   )
   &ScaleColor(
    Scale RGB=&VectorFromFloat(Value=&FloatFromCurve(
      FloatCurve=DataInterface() {propertyValues {Curve {keys [
          {interpMode=RCIM_Cubic, tangentMode=RCTM_User, time=0, value=0}
          {interpMode=RCIM_Cubic, tangentMode=RCTM_User, time=0.5, value=1}
          {tangentMode=RCTM_User, time=1, value=0}
         ]}}}
      Scale Curve=8
     ))
    ScaleA=true
    ScaleRGB=true
   ) {module=ScaleColor001}
   &Drag(Use Linear Drag=true, Use Rotational Drag=false)
   &SolveForcesAndVelocity(Clamp Velocity=false, Limit Acceleration=false)
  ]
  rendererClasses [&NiagaraLightRendererProperties]
  renderers [{rendererClass=&NiagaraLightRendererProperties, rendererIndex=0}]
 }
 InsectSwarm_Idle_Boundary {
  simTarget=GPUComputeSim
  emitterUpdateScript [
   &EmitterState()
   &SpawnPerUnit(Module.Use Max Movement Threshold=true, Module.Use Movement Tolerance=false, Spawn Spacing=50, Use Spawn Probability=false)
   &SpawnRate(Spawn Probability=&Lerp_Float(A=1, B=0, Alpha=$User.Flower), SpawnRate=5, Use Spawn Probability=true) {enabled=false, module=SpawnRate001}
  ]
  particleSpawnScript [
   &InitializeParticle(
    Color Mode=::NewEnumerator0
    Lifetime=&Multiply_Float(
     A=&RandomRangeFloat(Minimum=2, Maximum=3)
     B=&Lerp_Float(A=1, B=0.7, Alpha=$System.Hostile)
    )
    Mesh Scale Mode=::NewEnumerator1
    Mesh Uniform Scale Max=0.4
    Mesh Uniform Scale Min=0.3
    Sprite Size Mode=::NewEnumerator3
    Uniform Sprite Size=45
    UsePositionOffset=false
    Write Material Random=true
   )
   /SetVariables(Particles.MeshVisbilityIndex:NiagaraInt32=&SpawnGroup()) {isSetParameters=true}
   &ShapeLocation(
    MaskSpawnGroup=false
    Override Seed=false
    Sphere Radius=&RandomRangeFloat(Minimum=400, Maximum=400)
    UseOwnerScale=true
   )
  ]
  particleUpdateScript [
   &ParticleState()
   &CurlNoiseForce(
    Noise Frequency=12
    Noise Strength=&Lerp_Float(A=350, B=3000, Alpha=$User.Hostile)
    Pan Noise=true
    Pan Noise Field=<1 1 1>
   )
   &CurlNoiseForce(Noise Frequency=0.5, Noise Strength=200, Pan Noise=true, Pan Noise Field=<0.3 0.3 0.3>) {module=CurlNoiseForce001}
   &PointAttractionForce(
    Attraction Radius=800
    Attraction Strength=&Multiply_Float(
     A=&RandomRangeFloat(Minimum=200, Maximum=250)
     B=&Multiply_Float(
      A=&Multiply_Float(
       A=&Lerp_Float(
        A=0.2
        B=0.6
        Alpha=&Waveform("Waveform [1]"=::NewEnumerator9, "[1] Frequency"=0.4, "[1] Amplitude Min/Max"=<0 1>)
       )
       B=&Lerp_Float(A=0.7, B=1.5, Alpha=$User.Hostile)
      )
      B=&Lerp_Float(
       A=1
       B=0
       Alpha=&FloatFromCurve(
        FloatCurve=DataInterface() {propertyValues {Curve {keys [
            {arriveTangent=5.3671, interpMode=RCIM_Cubic, leaveTangent=5.3671, tangentMode=RCTM_User, time=0, value=0}
            {arriveTangent=1.3895, interpMode=RCIM_Cubic, leaveTangent=1.3895, tangentMode=RCTM_User, time=0.1574, value=0.7636}
            {arriveTangent=0.4231, interpMode=RCIM_Cubic, leaveTangent=0.4231, tangentMode=RCTM_User, time=1, value=1}
           ]}}}
        CurveIndex=$System.Timeline.Time
       )
      )
     )
    )
    Falloff Exponent=0.6
    Kill Within Radius=false
    Module.Use Falloff=true
    UseAttractorOffset=true
   )
   &PointForce(
    Force Origin:NiagaraPosition=$User.ShootPosition
    Force Strength=&Lerp_Float(A=0, B=5000, Alpha=$System.Timeline.Float)
    Module.Use Falloff=false
    Module.UseLinearForce=false
   )
   &ScaleSpriteSize(Uniform Curve Sprite Scale=DataInterface() {propertyValues {Curve {keys [
        {interpMode=RCIM_Cubic, tangentMode=RCTM_User, time=0.2, value=0}
        <at 0.3 1>
        {interpMode=RCIM_Cubic, tangentMode=RCTM_User, time=0.9, value=1}
        {tangentMode=RCTM_User, time=1, value=0}
       ]}}})
   &ScaleMeshSize(Scale Factor=&VectorFromFloat(Value=&FloatFromCurve(
      FloatCurve=DataInterface() {propertyValues {Curve {keys [
          {interpMode=RCIM_Cubic, tangentMode=RCTM_User, time=0, value=0}
          <at 0.3 1>
          {interpMode=RCIM_Cubic, tangentMode=RCTM_User, time=0.9, value=1}
          {tangentMode=RCTM_User, time=1, value=0}
         ]}}}
      Scale Curve=4
     )))
   &InterpolateOverTime(Rate Of Change=0.5, Target Float Value:NiagaraFloat=$User.Flower)
   &ScaleColor(
    Scale Alpha=0.4
    Scale RGB=&Lerp_Vector(
     A=&Lerp_Vector(A=<0 0 0>, B=<0 0.1 0.2>, Alpha=$StackContext.InterpolateOverTime.InterpolateOverTime_Float.Moving Average)
     B=<0.5 0 0>
     Alpha=$System.Hostile
    )
    ScaleA=true
    ScaleRGB=true
   ) {module=ScaleColor002}
   &ScaleColor(
    Scale RGB=&VectorFromFloat(Value=&FNTimeOfDayScalar(Night Scale=0.24, Morning Scale=0.4, Day Scale=0.9, Evening Scale=0.6))
    ScaleA=true
    ScaleRGB=true
   ) {module=ScaleColor001}
   &Drag(Use Linear Drag=true, Use Rotational Drag=false)
   &SolveForcesAndVelocity(Clamp Velocity=false, Limit Acceleration=false)
  ]
  rendererClasses [&NiagaraMeshRendererProperties, &NiagaraSpriteRendererProperties]
  renderers [
   {rendererClass=&NiagaraMeshRendererProperties, rendererIndex=0}
   {rendererClass=&NiagaraSpriteRendererProperties, rendererIndex=1}
  ]
 }
 InsectSwarm_Idle_LOW {
  emitterUpdateScript [
   &EmitterState()
   &SpawnRate(
    Spawn Probability=&Lerp_Float(A=1, B=0.5, Alpha=$System.SizeDistanceScalar)
    SpawnRate=60
    Use Spawn Probability=true
   )
   &SpawnRate(SpawnGroup=1, SpawnRate=3, Use Spawn Probability=false) {module=SpawnRate001}
  ]
  particleSpawnScript [
   &InitializeParticle(
    Color Mode=::NewEnumerator0
    Lifetime=&Multiply_Float(
     A=&RandomRangeFloat(Minimum=2, Maximum=3)
     B=&Lerp_Float(A=1, B=0.7, Alpha=$System.Hostile)
    )
    Mesh Scale Mode=::NewEnumerator1
    Mesh Uniform Scale Max=0.4
    Mesh Uniform Scale Min=0.2
    Sprite Size Mode=::NewEnumerator3
    Sprite UV Mode=::NewEnumerator1
    Uniform Sprite Size=45
    UsePositionOffset=false
    Write Material Random=true
   )
   /SetVariables(Particles.WaspSpriteSize:Vector2f=&Lerp_Vector2(A=<5 5>, B=<8 8>, Alpha=&RandomRangeFloat(Minimum=0.3, Maximum=1))) {isSetParameters=true, module=SetVariables_#7}
   /SetVariables(Particles.MeshVisbilityIndex:NiagaraInt32=&SpawnGroup()) {isSetParameters=true, module=SetVariables_#8}
   &ShapeLocation(
    MaskSpawnGroup=false
    Override Seed=false
    Sphere Radius=&RandomRangeFloat(Minimum=80, Maximum=120)
    UseOwnerScale=true
   )
  ]
  particleUpdateScript [
   &ParticleState()
   &CurlNoiseForce(
    Noise Frequency=12
    Noise Strength=&Lerp_Float(A=&Lerp_Float(A=350, B=100, Alpha=$User.Flower), B=3000, Alpha=$User.Hostile)
    Pan Noise=true
    Pan Noise Field=&Lerp_Vector(A=<1 1 1>, B=<0.2 0.2 0.2>, Alpha=$User.Flower)
   )
   &CurlNoiseForce(
    Noise Frequency=0.5
    Noise Strength=&Lerp_Float(A=500, B=0, Alpha=$User.Flower)
    Pan Noise=true
    Pan Noise Field=<0.3 0.3 0.3>
   ) {module=CurlNoiseForce001}
   &PointAttractionForce(
    Attraction Radius=800
    Attraction Strength=&Multiply_Float(
     A=&Multiply_Float(
      A=&RandomRangeFloat(Minimum=200, Maximum=250)
      B=&Lerp_Float(A=1, B=0.12, Alpha=$User.Flower)
     )
     B=&Multiply_Float(
      A=&Multiply_Float(
       A=&Lerp_Float(
        A=0.2
        B=0.6
        Alpha=&Waveform("Waveform [1]"=::NewEnumerator9, "[1] Frequency"=0.4, "[1] Amplitude Min/Max"=<0 1>)
       )
       B=&Lerp_Float(A=0.7, B=4, Alpha=$User.Hostile)
      )
      B=&Lerp_Float(
       A=1
       B=0
       Alpha=&FloatFromCurve(
        FloatCurve=DataInterface() {propertyValues {Curve {keys [
            {arriveTangent=5.3671, interpMode=RCIM_Cubic, leaveTangent=5.3671, tangentMode=RCTM_User, time=0, value=0}
            {arriveTangent=1.3895, interpMode=RCIM_Cubic, leaveTangent=1.3895, tangentMode=RCTM_User, time=0.1574, value=0.7636}
            {arriveTangent=0.4231, interpMode=RCIM_Cubic, leaveTangent=0.4231, tangentMode=RCTM_User, time=1, value=1}
           ]}}}
        CurveIndex=$System.Timeline.Time
       )
      )
     )
    )
    Falloff Exponent=0.6
    Kill Within Radius=false
    Module.Use Falloff=true
    UseAttractorOffset=true
   )
   &PointForce(
    Force Origin:NiagaraPosition=$User.ShootPosition
    Force Strength=&Lerp_Float(A=0, B=5000, Alpha=$System.Timeline.Float)
    Module.Use Falloff=false
    Module.UseLinearForce=false
   )
   &InterpolateOverTime(Rate Of Change=0.5, Target Float Value:NiagaraFloat=$User.Flower)
   &ScaleColor(
    Scale Alpha=0.3
    Scale RGB=&Lerp_Vector(
     A=&Lerp_Vector(A=<0 0 0>, B=<0 0.1 0.2>, Alpha=$StackContext.InterpolateOverTime.InterpolateOverTime_Float.Moving Average)
     B=<0.5 0 0>
     Alpha=$System.Hostile
    )
    ScaleA=true
    ScaleRGB=true
   )
   &ScaleColor(
    Scale RGB=&VectorFromFloat(Value=&FNTimeOfDayScalar(Night Scale=0.3, Morning Scale=0.6, Evening Scale=0.6))
    ScaleA=true
    ScaleRGB=true
   ) {module=ScaleColor001}
   &Drag(Use Linear Drag=true, Use Rotational Drag=false)
   &InterpolateOverTime(Rate Of Change=4, Target Float Value:NiagaraFloat=$User.AttractToHand) {module=InterpolateOverTime001}
   &ScaleMeshSize(Scale Factor=&VectorFromFloat(Value=&FloatFromCurve(
      FloatCurve=DataInterface() {propertyValues {Curve {keys [
          {interpMode=RCIM_Cubic, tangentMode=RCTM_User, time=0, value=0}
          <at 0.3 1>
          {interpMode=RCIM_Cubic, tangentMode=RCTM_User, time=0.9, value=1}
          {tangentMode=RCTM_User, time=1, value=0}
         ]}}}
      Scale Curve=&Lerp_Float(A=4, B=0, Alpha=$StackContext.InterpolateOverTime001.InterpolateOverTime_Float.Moving Average)
     )))
   &ScaleSpriteSize(
    Uniform Curve Scale=&Multiply_Float(
     A=&Lerp_Float(A=1, B=5, Alpha=$System.SizeDistanceScalar)
     B=&Lerp_Float(A=1, B=0, Alpha=$StackContext.InterpolateOverTime001.InterpolateOverTime_Float.Moving Average)
    )
    Uniform Curve Sprite Scale=DataInterface() {propertyValues {Curve {keys [
        {interpMode=RCIM_Cubic, tangentMode=RCTM_User, time=0.2, value=0}
        <at 0.3 1>
        {interpMode=RCIM_Cubic, tangentMode=RCTM_User, time=0.9, value=1}
        {tangentMode=RCTM_User, time=1, value=0}
       ]}}}
   )
   /SetVariables(Particles.WaspSpriteSize:Vector2f=&Multiply_Vector2D(
     A=$Particles.Initial.WaspSpriteSize
     B=&Multiply_Vector2D_ByFloat(
      Float=&FloatFromCurve(
       FloatCurve=DataInterface() {propertyValues {Curve {keys [
           <at 0 0 RCIM_Cubic>
           {arriveTangent=0.2423, interpMode=RCIM_Cubic, leaveTangent=0.2423, tangentMode=RCTM_User, time=0.2, value=1}
           {arriveTangent=0.0664, interpMode=RCIM_Cubic, leaveTangent=0.0664, tangentMode=RCTM_User, time=0.8, value=1}
           <at 1 0 RCIM_Cubic>
          ]}}}
       Scale Curve=&Lerp_Float(A=1, B=2, Alpha=$System.Hostile)
      )
      Vector2D=&Multiply_Vector2D_ByFloat(
       Float=&Lerp_Float(A=&RandomRangeFloat(Minimum=1.5, Maximum=3), B=3, Alpha=$System.SizeDistanceScalar)
       Vector2D=&Multiply_Vector2D_ByFloat(
        Float=&Lerp_Float(A=1, B=0, Alpha=$StackContext.InterpolateOverTime001.InterpolateOverTime_Float.Moving Average)
        Vector2D=<1 1>
       )
      )
     )
    )) {isSetParameters=true}
   &LerpParticleAttributes(
    Lerp Alpha:NiagaraFloat=$StackContext.InterpolateOverTime001.InterpolateOverTime_Float.Moving Average
    Lerp Position=true
    Position B:NiagaraPosition=$User.CollectHandPosition
   )
   &SolveForcesAndVelocity(Clamp Velocity=false, Limit Acceleration=false)
   &DynamicMaterialParameters(
    Index 0 Param 1=$User.Flower
    Index 0 Param 2=&Lerp_Float(A=1, B=0.4, Alpha=$User.Flower)
    Module.Param0WriteEnabled=true
    Module.Param1WriteEnabled=true
    Module.Param2WriteEnabled=true
    Module.Param3WriteEnabled=true
   ) {module=DynamicMaterialParameters001}
   &KillParticles(Kill Particles=&SetBoolByFloatComparison(A=$Particles.InterpolateOverTime001.InterpolateOverTime_Float.Moving Average, B=0.5))
  ]
  rendererClasses [&NiagaraMeshRendererProperties, &NiagaraSpriteRendererProperties]
  renderers [
   {rendererClass=&NiagaraMeshRendererProperties, rendererIndex=0}
   {rendererClass=&NiagaraSpriteRendererProperties, rendererIndex=1}
   {rendererClass=&NiagaraSpriteRendererProperties, rendererIndex=2}
  ]
 }
 InsectSwarm_Idle_Turning {
  bEnabled=false
  simTarget=GPUComputeSim
  emitterUpdateScript [
   &EmitterState()
   &SpawnRate(
    Spawn Probability=&Lerp_Float(A=1, B=1, Alpha=$System.SizeDistanceScalar)
    SpawnRate=60
    Use Spawn Probability=true
   )
   &SpawnRate(SpawnGroup=1, SpawnRate=5, Use Spawn Probability=false) {module=SpawnRate001}
  ]
  particleSpawnScript [
   &InitializeParticle(
    Color Mode=::NewEnumerator0
    Lifetime=&Multiply_Float(
     A=&RandomRangeFloat(Minimum=2, Maximum=3)
     B=&Lerp_Float(A=1, B=0.7, Alpha=$System.Hostile)
    )
    Mesh Scale Mode=::NewEnumerator1
    Mesh Uniform Scale Max=0.4
    Mesh Uniform Scale Min=0.2
    Sprite Size Mode=::NewEnumerator3
    Uniform Sprite Size=150
    UsePositionOffset=false
    Write Material Random=true
   )
   /SetVariables(Particles.WaspSpriteSize:Vector2f=&Lerp_Vector2(A=<5 5>, B=<8 8>, Alpha=&RandomRangeFloat(Minimum=0.3, Maximum=1))) {isSetParameters=true, module=SetVariables_#7}
   /SetVariables(Particles.MeshVisbilityIndex:NiagaraInt32=&SpawnGroup()) {isSetParameters=true, module=SetVariables_#8}
   &ShapeLocation(
    MaskSpawnGroup=false
    Override Seed=false
    Sphere Radius=&RandomRangeFloat(Minimum=80, Maximum=120)
    UseOwnerScale=true
   )
  ]
  particleUpdateScript [
   /SetVariables(Particles.WaveFormCurve:NiagaraFloat=&Waveform(
     "Waveform [1]"=::NewEnumerator0
     "[1] Phase"=&Add_Float(A=$Emitter.Age, B=0)
     "[1] Frequency"=1
     "[1] Amplitude Min/Max"=<0 1>
    )) {isSetParameters=true}
   &ParticleState()
   &CurlNoiseForce(
    Noise Frequency=12
    Noise Strength=&Lerp_Float(A=&Lerp_Float(A=350, B=100, Alpha=$User.Flower), B=3000, Alpha=$User.Hostile)
    Pan Noise=true
    Pan Noise Field=&Lerp_Vector(A=<1 1 1>, B=<0.2 0.2 0.2>, Alpha=$User.Flower)
   ) {enabled=false}
   &CurlNoiseForce(
    Noise Frequency=0.5
    Noise Strength=&Lerp_Float(A=500, B=0, Alpha=$User.Flower)
    Pan Noise=true
    Pan Noise Field=<0.3 0.3 0.3>
   ) {enabled=false, module=CurlNoiseForce001}
   &PointAttractionForce(
    Attraction Radius=800
    Attraction Strength=&Multiply_Float(
     A=&Multiply_Float(
      A=&RandomRangeFloat(Minimum=200, Maximum=250)
      B=&Lerp_Float(A=1, B=0.12, Alpha=$User.Flower)
     )
     B=&Multiply_Float(
      A=&Multiply_Float(
       A=&Lerp_Float(
        A=0.2
        B=0.6
        Alpha=&Waveform("Waveform [1]"=::NewEnumerator9, "[1] Frequency"=0.4, "[1] Amplitude Min/Max"=<0 1>)
       )
       B=&Lerp_Float(A=0.7, B=4, Alpha=$User.Hostile)
      )
      B=&Lerp_Float(
       A=1
       B=0
       Alpha=&FloatFromCurve(
        FloatCurve=DataInterface() {propertyValues {Curve {keys [
            {arriveTangent=5.3671, interpMode=RCIM_Cubic, leaveTangent=5.3671, tangentMode=RCTM_User, time=0, value=0}
            {arriveTangent=1.3895, interpMode=RCIM_Cubic, leaveTangent=1.3895, tangentMode=RCTM_User, time=0.1574, value=0.7636}
            {arriveTangent=0.4231, interpMode=RCIM_Cubic, leaveTangent=0.4231, tangentMode=RCTM_User, time=1, value=1}
           ]}}}
        CurveIndex=$System.Timeline.Time
       )
      )
     )
    )
    Falloff Exponent=0.6
    Kill Within Radius=false
    Module.Use Falloff=true
    UseAttractorOffset=true
   ) {enabled=false}
   &PointForce(
    Force Origin:NiagaraPosition=$User.ShootPosition
    Force Strength=&Lerp_Float(A=0, B=5000, Alpha=$System.Timeline.Float)
    Module.Use Falloff=false
    Module.UseLinearForce=false
   ) {enabled=false}
   &InterpolateOverTime(Rate Of Change=0.5, Target Float Value:NiagaraFloat=$User.Flower)
   &ScaleColor(
    Scale Alpha=0.4
    Scale RGB=&Lerp_Vector(
     A=&Lerp_Vector(A=<0 0 0>, B=<0 0.2 0.4>, Alpha=$StackContext.InterpolateOverTime.InterpolateOverTime_Float.Moving Average)
     B=<0.5 0 0>
     Alpha=$System.Hostile
    )
    ScaleA=true
    ScaleRGB=true
   )
   &ScaleColor(
    Scale RGB=&VectorFromFloat(Value=&FNTimeOfDayScalar(Night Scale=0.24, Morning Scale=0.4, Day Scale=0.9, Evening Scale=0.6))
    ScaleA=true
    ScaleRGB=true
   ) {module=ScaleColor001}
   &Drag(Use Linear Drag=true, Use Rotational Drag=false)
   &InterpolateOverTime(Rate Of Change=4, Target Float Value:NiagaraFloat=$User.AttractToHand) {module=InterpolateOverTime001}
   &ScaleMeshSize(Scale Factor=&VectorFromFloat(Value=&FloatFromCurve(
      FloatCurve=DataInterface() {propertyValues {Curve {keys [
          {interpMode=RCIM_Cubic, tangentMode=RCTM_User, time=0, value=0}
          <at 0.3 1>
          {interpMode=RCIM_Cubic, tangentMode=RCTM_User, time=0.9, value=1}
          {tangentMode=RCTM_User, time=1, value=0}
         ]}}}
      Scale Curve=&Lerp_Float(A=4, B=0, Alpha=$StackContext.InterpolateOverTime001.InterpolateOverTime_Float.Moving Average)
     )))
   &ScaleSpriteSize(
    Uniform Curve Scale=&Multiply_Float(
     A=&Lerp_Float(A=1, B=5, Alpha=$System.SizeDistanceScalar)
     B=&Lerp_Float(A=1, B=0, Alpha=$StackContext.InterpolateOverTime001.InterpolateOverTime_Float.Moving Average)
    )
    Uniform Curve Sprite Scale=DataInterface() {propertyValues {Curve {keys [
        {interpMode=RCIM_Cubic, tangentMode=RCTM_User, time=0.2, value=0}
        <at 0.3 1>
        {interpMode=RCIM_Cubic, tangentMode=RCTM_User, time=0.9, value=1}
        {tangentMode=RCTM_User, time=1, value=0}
       ]}}}
   )
   /SetVariables(Particles.WaspSpriteSize:Vector2f=&Multiply_Vector2D(
     A=$Particles.Initial.WaspSpriteSize
     B=&Multiply_Vector2D_ByFloat(
      Float=&FloatFromCurve(
       FloatCurve=DataInterface() {propertyValues {Curve {keys [
           <at 0 0 RCIM_Cubic>
           {arriveTangent=0.2423, interpMode=RCIM_Cubic, leaveTangent=0.2423, tangentMode=RCTM_User, time=0.2, value=1}
           {arriveTangent=0.0664, interpMode=RCIM_Cubic, leaveTangent=0.0664, tangentMode=RCTM_User, time=0.8, value=1}
           <at 1 0 RCIM_Cubic>
          ]}}}
       Scale Curve=&Lerp_Float(A=1, B=1.5, Alpha=$System.Hostile)
      )
      Vector2D=&Multiply_Vector2D_ByFloat(
       Float=&Lerp_Float(A=&RandomRangeFloat(Minimum=1.5, Maximum=3), B=4, Alpha=$System.SizeDistanceScalar)
       Vector2D=&Multiply_Vector2D_ByFloat(
        Float=&Lerp_Float(A=1, B=0, Alpha=$StackContext.InterpolateOverTime001.InterpolateOverTime_Float.Moving Average)
        Vector2D=<1 1>
       )
      )
     )
    )) {isSetParameters=true}
   &LerpParticleAttributes(
    Lerp Alpha:NiagaraFloat=$StackContext.InterpolateOverTime001.InterpolateOverTime_Float.Moving Average
    Lerp Position=true
    Position B:NiagaraPosition=$User.CollectHandPosition
   )
   &SolveForcesAndVelocity(Clamp Velocity=false, Limit Acceleration=false)
   &DynamicMaterialParameters(
    Index 0 Param 1=$User.Flower
    Index 0 Param 2=&Lerp_Float(A=1, B=0.4, Alpha=$User.Flower)
    Index 0 Param 3=&ClampFloat(Float=&FloatFromCurve(
      FloatCurve=DataInterface() {propertyValues {Curve {keys [
          {tangentMode=RCTM_User, time=0, value=1}
          <at 0.4136 1>
          <at 0.4408 0>
          <at 0.6535 0>
          <at 0.6895 1>
          {tangentMode=RCTM_User, time=1, value=1}
         ]}}}
      CurveIndex=$Particles.WaveFormCurve
     ))
    Index 0 Param 4=&ClampFloat(
     Float=&FloatFromCurve(
      FloatCurve=DataInterface() {propertyValues {Curve {keys [
          {tangentMode=RCTM_User, time=0, value=0}
          {arriveTangent=1.6667, interpMode=RCIM_Cubic, leaveTangent=1.6667, time=0.4, value=0}
          <at 0.6 1>
          {tangentMode=RCTM_User, time=1, value=1}
         ]}}}
      CurveIndex=$Particles.WaveFormCurve
     )
     Max=0.98
    )
    Index 1 Param 1=&ClampFloat(Float=&FloatFromCurve(
      FloatCurve=DataInterface() {propertyValues {Curve {keys [{tangentMode=RCTM_User, time=0, value=1}, <at 0.4 1 RCIM_Constant>, <at 0.6 0>, {tangentMode=RCTM_User, time=1, value=0}]}}}
      CurveIndex=$Particles.WaveFormCurve
     ))
    Module.Param0WriteEnabled=true
    Module.Param0WriteEnabled1=true
    Module.Param1WriteEnabled=true
    Module.Param1WriteEnabled1=true
    Module.Param2WriteEnabled=true
    Module.Param2WriteEnabled1=true
    Module.Param3WriteEnabled=true
    Module.Param3WriteEnabled1=true
    Write Parameter Index 1=true
   ) {module=DynamicMaterialParameters001}
   &KillParticles(Kill Particles=&SetBoolByFloatComparison(A=$Particles.InterpolateOverTime001.InterpolateOverTime_Float.Moving Average, B=0.5))
  ]
  rendererClasses [&NiagaraMeshRendererProperties, &NiagaraSpriteRendererProperties]
  renderers [
   {rendererClass=&NiagaraMeshRendererProperties, rendererIndex=0}
   {rendererClass=&NiagaraSpriteRendererProperties, rendererIndex=1}
   {rendererClass=&NiagaraSpriteRendererProperties, rendererIndex=2}
  ]
 }
 InsectSwarm_Trail {
  simTarget=GPUComputeSim
  emitterUpdateScript [
   &EmitterState(Inactive Response=::NewEnumerator1, Life Cycle Mode=::NewEnumerator1, Loop Duration=1, UseLoopDelay=false)
   &SpawnPerUnit(Module.Use Max Movement Threshold=true, Module.Use Movement Tolerance=false, Spawn Spacing=4, Use Spawn Probability=false)
  ]
  particleSpawnScript [
   &InitializeParticle(
    Color Mode=::NewEnumerator0
    Lifetime Mode=::NewEnumerator1
    Sprite Size Mode=::NewEnumerator1
    Sprite UV Mode=::NewEnumerator1
    UsePositionOffset=false
    Write Material Random=true
   )
   /SetVariables(Particles.RandomInsectTexture:NiagaraFloat=&UniformAOrBFloat(A=0)) {isSetParameters=true, module=SetVariables_#6}
   /SetVariables(Particles.WaspSpriteSize:Vector2f=&Lerp_Vector2(A=<5 5>, B=<8 8>, Alpha=&RandomRangeFloat(Minimum=0.5, Maximum=1))) {isSetParameters=true, module=SetVariables_#7}
   &ShapeLocation(
    MaskSpawnGroup=false
    Override Seed=false
    Sphere Radius=&RandomRangeFloat(Minimum=10, Maximum=120)
    UseOwnerScale=true
   )
   &DynamicMaterialParameters(
    Index 0 Param 4=$Particles.RandomInsectTexture
    Module.Param0WriteEnabled=false
    Module.Param1WriteEnabled=false
    Module.Param2WriteEnabled=false
    Module.Param3WriteEnabled=true
   )
  ]
  particleUpdateScript [
   &ParticleState()
   &CurlNoiseForce(Noise Frequency=12, Noise Strength=1000, Pan Noise=true, Pan Noise Field=<0.5 0.5 0.5>)
   &CurlNoiseForce(Noise Frequency=1, Noise Strength=300, Pan Noise=true, Pan Noise Field=<0.3 0.3 0.3>) {module=CurlNoiseForce001}
   &PointAttractionForce(
    Attraction Radius=800
    Attraction Strength=120
    Falloff Exponent=0.55
    Kill Within Radius=false
    Module.Use Falloff=true
    UseAttractorOffset=true
   ) {module=PointAttractionForce001}
   &InheritVelocity(Inherited Velocity Amount Scale=<1.2 1.2 1.2>, Inherited Velocity Speed Limit=6, Speed Limit=true, UseSpeedThreshold=false)
   &PointForce(
    Force Origin:NiagaraPosition=$User.ShootPosition
    Force Strength=&Lerp_Float(A=0, B=20000, Alpha=$System.Timeline.Float)
    Module.Use Falloff=false
    Module.UseLinearForce=false
   )
   /SetVariables(Particles.WaspSpriteSize:Vector2f=&Multiply_Vector2D(
     A=$Particles.Initial.WaspSpriteSize
     B=&Multiply_Vector2D_ByFloat(
      Float=&FloatFromCurve(
       FloatCurve=DataInterface() {propertyValues {Curve {keys [
           <at 0 0 RCIM_Cubic>
           {arriveTangent=0.2423, interpMode=RCIM_Cubic, leaveTangent=0.2423, tangentMode=RCTM_User, time=0.2, value=1}
           {arriveTangent=0.0664, interpMode=RCIM_Cubic, leaveTangent=0.0664, tangentMode=RCTM_User, time=0.8, value=1}
           <at 1 0 RCIM_Cubic>
          ]}}}
       Scale Curve=&Lerp_Float(A=1, B=1.5, Alpha=$System.Hostile)
      )
      Vector2D=&Multiply_Vector2D_ByFloat(
       Float=&Lerp_Float(A=&RandomRangeFloat(Minimum=1.3, Maximum=2), B=3.5, Alpha=$System.SizeDistanceScalar)
       Vector2D=<1 1>
      )
     )
    )) {isSetParameters=true}
   &Drag(Drag=8, Use Linear Drag=true, Use Rotational Drag=true)
   &SolveForcesAndVelocity(Clamp Velocity=false, Limit Acceleration=false)
  ]
  rendererClasses [&NiagaraSpriteRendererProperties]
  renderers [{rendererClass=&NiagaraSpriteRendererProperties}]
 }
}
summary {
 systemName=NS_AI_Wildlife_InsectSwarm_Idle
 emitters [
  {emitterName=InsectSwarm_Trail, simTarget=GPUComputeSim, rendererClasses [&NiagaraSpriteRendererProperties]}
  {emitterName=InsectSwarm_Idle001, rendererClasses [&NiagaraLightRendererProperties]}
  {emitterName=GroundDust, rendererClasses [&NiagaraSpriteRendererProperties]}
  {
   emitterName=InsectSwarm_Idle_Boundary
   simTarget=GPUComputeSim
   rendererClasses [&NiagaraMeshRendererProperties, &NiagaraSpriteRendererProperties]
  }
  {emitterName=FakeSawrm_LOW, rendererClasses [&NiagaraMeshRendererProperties]}
  {emitterName=FakeSawrm_Trail_LOW, rendererClasses [&NiagaraMeshRendererProperties]}
  {
   emitterName=InsectSwarm_Idle
   simTarget=GPUComputeSim
   rendererClasses [&NiagaraMeshRendererProperties, &NiagaraSpriteRendererProperties]
  }
  {
   emitterName=InsectSwarm_Idle_LOW
   rendererClasses [&NiagaraMeshRendererProperties, &NiagaraSpriteRendererProperties]
  }
  {
   bEnabled=false
   emitterName=InsectSwarm_Idle_Turning
   simTarget=GPUComputeSim
   rendererClasses [&NiagaraMeshRendererProperties, &NiagaraSpriteRendererProperties]
  }
 ]
 userVariables [
  User.AttractToHand {type=NiagaraFloat, defaultValue=0, description=""}
  User.ByShoot {type=NiagaraBool}
  User.CollectHandPosition {type=NiagaraPosition, description="", defaultValue {}}
  User.DustLocation {type=NiagaraPosition, description="", defaultValue {}}
  User.Flower {type=NiagaraFloat, defaultValue=0, description=""}
  User.Hostile {type=NiagaraFloat, defaultValue=0, description=""}
  User.Radius {type=NiagaraFloat, defaultValue=100, description=""}
  User.ShootPosition {type=NiagaraPosition, description="", defaultValue {}}
 ]
}
