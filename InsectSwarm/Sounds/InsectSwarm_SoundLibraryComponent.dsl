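; Sound library component for the insect swarm that drives its looping audio from hostile and on-plant states.
; It caches the Loop event's audio component and sets its OnPlant parameter directly.
; Hostility changes start a repeating timer that interpolates the IsHostile parameter until it reaches either extreme, and turning hostile also plays the Threatened event.
; Stopped sounds fade out and clear that timer.
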(event EventBeginPlay)

(event EventTick
  (params (DeltaSeconds :type Float%20%28single-precision%29)))

(fn OnSoundPlayed
  (params (InEventName :type Gameplay%20Tag%20Structure) (InComponent :type Audio%20Component%20Object%20Reference))
  (if (== InEventName "(TagName=\"SoundLibrary.InsectSwarm.Loop\")")
    (Variables|Default|SetLoopAudio InComponent)))

(fn OnSoundStopped
  (params (InEventName :type Gameplay%20Tag%20Structure) (InComponent :type Audio%20Component%20Object%20Reference))
  (Audio|Components|Audio|FadeOut InComponent 1.0 0.0 "Linear" :!class /Script/Engine.AudioComponent)
  (CallFunction|StopHostileHandle)
  (return true))

(fn StopHostileHandle
  (bind _hostilehandle (Variables|Default|GetHostileHandle))
  (if (Utilities|Time|IsValidTimerHandle _hostilehandle)
    (Utilities|Time|ClearandInvalidateTimerbyHandle _hostilehandle)))

(fn UpdateHostileParam
  (bind _output_get (Variables|Default|SetHostileParam (Math|Interpolation|FInterpTo (Variables|Default|GetHostileParam) (Math|Conversions|ToFloat(Boolean) (Variables|Default|GetIsHostile)) (Utilities|Time|GetWorldDeltaSeconds) 1e+01)))
  (Audio|Parameter|SetFloatParameter (Variables|Default|GetLoopAudio) "IsHostile" _output_get :!class /Script/AudioExtensions.AudioParameterControllerInterface)
  (if (or (== (Variables|Default|GetHostileParam) 1.0) (== (Variables|Default|GetHostileParam) 0))
    (CallFunction|StopHostileHandle)))

(fn UpdateHostileStatus
  (params (IsHostile :type Boolean))
  (Variables|Default|SetIsHostile IsHostile)
  (bind _returnvalue (Utilities|Time|SetTimerbyEvent (EventDispatchers|CreateEvent :!event UpdateHostileParam) 0.1 true))
  (Variables|Default|SetHostileHandle _returnvalue)
  (if (Variables|Default|GetIsHostile)
    (SoundLibrary|PlayEvent "(TagName=\"SoundLibrary.InsectSwarm.Threatened\")")))

(fn UpdatePlantStatus
  (params (IsOnPlant :type Boolean))
  (Audio|Parameter|SetFloatParameter (Variables|Default|GetLoopAudio) "OnPlant" (Math|Conversions|ToFloat(Boolean) IsOnPlant) :!class /Script/AudioExtensions.AudioParameterControllerInterface))
