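; Sound Library context Blueprint for InsectSwarm that handles playing a sound for an event.
; Its Play function breaks the event input struct into a sound and an owning actor, then spawns the sound attached to the owner's root component.
; It returns the spawned audio component in an array along with a success value of true.
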(fn Play
  (params (InEventData :type Sound%20Library%20Context%20Event%20Input%20Structure%20%28by%20ref%29))
  (bind (_sound _owner) (Utilities|Struct|BreakSoundLibraryContextEventInput InEventData))
  (bind _returnvalue (Audio|SpawnSoundAttached _sound (Variables|/Script/Engine.Actor|GetRootComponent _owner)))
  (return (Utilities|Array|MakeArray _returnvalue) true))
